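# Material for the Hunt skin of the SpacePlunge pickaxe, layering animated emissive effects and base color darkening between the shared 2023 character PreFX and PostFX functions.
# A height gradient from pre-skinned position, modulated by a TechArt texture, blends two HDR colors for the glow.
# Three procedural square-cell layers project pre-skinned position by dominant normal axis and flicker through time-seeded hashes.
# Named reroutes carry these layers, texture-derived masks and a sine idle pulse.
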Material /BRCosmetics/Weapons/FORT_Melee/Pickaxe_SpacePlunge/Skins/Hunt/Materials/M_PickaxeSpacePlunge_Hunt
graph {
 properties {
  bUseMaterialAttributes=true
  bUsedWithMeshDeformer=true
  bUsedWithMorphTargets=true
  bUsedWithSkeletalMesh=true
  pixelDepthOffsetMode=PDOM_Legacy
  shadingModel=MSM_FromMaterialExpression
  usedShadingModels=Default Lit | Subsurface | Clear Coat | Subsurface Profile | Cloth
 }
 nodes [
  n1 = MaterialFunctionCall() {materialFunction=&MF_CharacterParent_2023_PreFX}
  n2 = GetMaterialAttributes(None=@n1.Result) {attributeGetTypes [#1, #2]}
  n3 = VectorParameter() {group=HuntReactivity, parameterName=CentralEM_High, defaultValue=<rgba 1 0.3125 0 10>}
  n4 = VectorParameter() {group=HuntReactivity, parameterName=CentralEM_Low, defaultValue=<rgba 1 0 0 2>}
  n5 = TextureSample() {texture=/BRCosmetics/Weapons/FORT_Melee/Pickaxe_SpacePlunge/Skins/Hunt/Textures/T_spaceplunge_Pickaxe_Pickaxe_TechArt01}
  n6 = Subtract(
   A=ComponentMask(None=@n1.Pre skinned Local Position) {b=true}
   B=ScalarParameter() {defaultValue=20, group=Hunt, parameterName=HeightGradient}
  )
  n7 = ScalarParameter() {defaultValue=20, group=Hunt, parameterName=HeightUnit}
  n8 = Multiply(A=Saturate(None=Divide(A=@n6, B=@n7)), B=LinearInterpolate(Alpha=@n5.R) {constA=1, constB=5})
  n9 = Multiply(
   A=LinearInterpolate(A=Multiply(A=@n3.RGB, B=@n3.RGB), B=Multiply(A=@n4.RGB, B=@n4.RGB), Alpha=OneMinus(None=@n8))
   B=LinearInterpolate(Alpha=@n5.G) {constA=1, constB=0.3}
  )
  n10 = LinearInterpolate(Alpha=NamedRerouteUsage()) {constA=0.5}
  n11 = Multiply(A=LinearInterpolate(Alpha=NamedRerouteUsage()) {constA=1, constB=0.1}, B=Multiply(A=NamedRerouteUsage(), B=Multiply(A=@n9, B=@n10)))
  n12 = Constant()
  n13 = SmoothStep(Value=NamedRerouteUsage()) {constMax=0.3}
  MaterialFunctionCall(
   Material=SetMaterialAttributes!(
    MaterialAttributes=@n2.MaterialAttributes
    Emissive Color=Add(A=@n2.Emissive Color, B=@n11)
    Base Color=LinearInterpolate(A=@n2.Base Color, B=@n12, Alpha=@n13)
   ) {attributeSetTypes [#1, #2]}
   M=@n1.M
   Pre Skinned Local Position=@n1.M
   Distance Blend=@n1.M
   Fresnel=@n1.M
  ) {materialFunction=&MF_CharacterParent_2023_PostFX}
  MaterialFunctionCall!(RotationAngle, NormalizedRotationAxis, PivotPoint, Position, Normal, TangentU, TangentV) {materialFunction=/BRCosmetics/Characters/Player/Male/Medium/Bodies/M_MED_Inferno/FX/Mat/RotateVertexAboutAxis}
  NamedRerouteDeclaration(None=ComponentMask(None=@n1.Tech Art Mask) {b=true}) {name=EmissiveMask, nodeColor=<rgba 0.875 0 1 1>}
  NamedRerouteDeclaration(None=SmoothStep(Value=OneMinus(None=@n5.B))) {name=GlitchMask, nodeColor=<rgba 0 0.9531 1 1>}
  n14 = Sine(None=Multiply(A=Time(), B=ScalarParameter() {defaultValue=0.3, group=HuntReactivity, parameterName=IdleTimeSpeed}))
  NamedRerouteDeclaration(None=SmoothStep(Value=ConstantBiasScale(None=@n14))) {name=GradientChangeAlpha, nodeColor=<rgba 1 0 0.1406 1>}
  n15 = Constant3Vector() {constant=<rgba 0 1 0 0>}
  n16 = ScalarParameter() {defaultValue=0.1, parameterName=RotateRadius}
  n17 = Constant3Vector()
  NamedRerouteDeclaration(None=Add(
    A=@n1.Pre skinned Local Position
    B=RotateAboutAxis(NormalizedRotationAxis=@n15, RotationAngle=@n16, PivotPoint=@n17, Position=@n1.Pre skinned Local Position)
   )) {name=PreskinnedPOS, nodeColor=<rgba 0 1 0.8125 1>}
  n18 = Multiply(A=NamedRerouteUsage(), B=ScalarParameter() {defaultValue=0.8, parameterName=CubeLayerEM})
  n19 = NamedRerouteUsage()
  n20 = Multiply(A=ScalarParameter() {defaultValue=1, parameterName=CubeLayerEM_02}, B=NamedRerouteUsage())
  NamedRerouteDeclaration(None=LinearInterpolate(A=LinearInterpolate(A=@n18, B=@n19, Alpha=@n20), B=NamedRerouteUsage(), Alpha=NamedRerouteUsage())) {name=PropEM, nodeColor=<rgba 0 1 0.4609 1>}
  n21 = Multiply(A=Time(), B=ScalarParameter() {defaultValue=0.6, parameterName=PumpSpeed_02})
  n22 = Add(A=NamedRerouteUsage(), B=VectorParameter() {group=Hunt, parameterName=OffsetUnit, defaultValue=<rgba 5 1 2 1>}.RGB)
  n23 = ComponentMask(None=@n22) {b=true, r=true}
  n24 = ScalarParameter() {defaultValue=3, parameterName=TintingUnit_02}
  n25 = ComponentMask(None=@n22) {b=true, g=true}
  n26 = VertexInterpolator(VS=PreSkinnedNormal())
  n27 = Abs(None=DotProduct(A=@n26.PS, B=Constant3Vector() {constant=<rgba 1 0 0 0>}))
  n28 = DotProduct(A=@n26.PS, B=Constant3Vector() {constant=<rgba 0 0 1 0>})
  n29 = LinearInterpolate(
   A=LinearInterpolate(A=Divide(A=@n23, B=@n24), B=Divide(A=@n25, B=@n24), Alpha=Step(X=@n27) {constY=0.5})
   B=Divide(A=ComponentMask(None=@n22) {g=true, r=true}, B=@n24)
   Alpha=Step(X=Abs(None=@n28)) {constY=0.5}
  )
  n30 = Floor(None=@n29)
  n31 = Abs(None=ComponentMask(None=MaterialFunctionCall(Seed=@n21, Vec2=@n30) {materialFunction=&MF_Hash}.Vec2) {r=true})
  n32 = Frac(None=@n29)
  n33 = Constant() {r=0.5}
  n34 = LinearInterpolate(Alpha=Saturate(None=@n31)) {constA=0.35, constB=0.49}
  n35 = SphereMask(A=ComponentMask(None=@n32) {r=true}, B=@n33, Radius=@n34) {hardnessPercent=0}
  n36 = SphereMask(A=ComponentMask(None=@n32) {g=true}, B=@n33, Radius=@n34) {hardnessPercent=0}
  n37 = Multiply(A=Sign(None=@n35), B=Sign(None=@n36))
  n38 = LinearInterpolate(Alpha=Multiply(A=SmoothStep(Value=@n35) {constMax=0.3}, B=SmoothStep(Value=@n36) {constMax=0.3})) {constA=0.5}
  NamedRerouteDeclaration(None=Multiply(A=@n31, B=Multiply(A=@n37, B=@n38))) {name=PropEM_1, nodeColor=<rgba 0 1 0.4609 1>}
  n39 = Multiply(A=Time(), B=ScalarParameter() {defaultValue=0.4, group=Hunt, parameterName=PumpSpeed_03})
  n40 = Add(A=NamedRerouteUsage(), B=VectorParameter() {group=Hunt, parameterName=OffsetUnit_02, defaultValue=<rgba 1 2 1 1>}.RGB)
  n41 = ComponentMask(None=@n40) {b=true, r=true}
  n42 = ScalarParameter() {defaultValue=4, parameterName=TintingUnit_03}
  n43 = ComponentMask(None=@n40) {b=true, g=true}
  n44 = VertexInterpolator(VS=PreSkinnedNormal())
  n45 = Abs(None=DotProduct(A=@n44.PS, B=Constant3Vector() {constant=<rgba 1 0 0 0>}))
  n46 = DotProduct(A=@n44.PS, B=Constant3Vector() {constant=<rgba 0 0 1 0>})
  n47 = LinearInterpolate(
   A=LinearInterpolate(A=Divide(A=@n41, B=@n42), B=Divide(A=@n43, B=@n42), Alpha=Step(X=@n45) {constY=0.5})
   B=Divide(A=ComponentMask(None=@n40) {g=true, r=true}, B=@n42)
   Alpha=Step(X=Abs(None=@n46)) {constY=0.5}
  )
  n48 = Floor(None=@n47)
  n49 = Abs(None=ComponentMask(None=MaterialFunctionCall(Seed=@n39, Vec2=@n48) {materialFunction=&MF_Hash}.Vec2) {r=true})
  n50 = Frac(None=@n47)
  n51 = Constant() {r=0.5}
  n52 = LinearInterpolate(Alpha=Saturate(None=@n49)) {constA=0.3, constB=0.45}
  n53 = SphereMask(A=ComponentMask(None=@n50) {r=true}, B=@n51, Radius=@n52) {hardnessPercent=0}
  n54 = SphereMask(A=ComponentMask(None=@n50) {g=true}, B=@n51, Radius=@n52) {hardnessPercent=0}
  n55 = Multiply(A=Sign(None=@n53), B=Sign(None=@n54))
  n56 = LinearInterpolate(Alpha=Multiply(A=SmoothStep(Value=@n53) {constMax=0.2}, B=SmoothStep(Value=@n54) {constMax=0.2})) {constA=0.5}
  NamedRerouteDeclaration(None=Multiply(A=@n49, B=Multiply(A=@n55, B=@n56))) {name=PropEM_2, nodeColor=<rgba 0 1 0.4609 1>}
  NamedRerouteDeclaration(None=SmoothStep(Value=Multiply(A=@n49, B=@n55)) {constMax=0.7}) {name=BlendAlpha02, nodeColor=<rgba 0 0.2031 1 1>}
  n57 = Multiply(A=Time(), B=ScalarParameter() {defaultValue=0.8, parameterName=PumpSpeed})
  n58 = NamedRerouteUsage()
  n59 = ComponentMask(None=@n58) {b=true, r=true}
  n60 = ScalarParameter() {defaultValue=2, parameterName=TintingUnit}
  n61 = ComponentMask(None=@n58) {b=true, g=true}
  n62 = VertexInterpolator(VS=PreSkinnedNormal())
  n63 = Abs(None=DotProduct(A=@n62.PS, B=Constant3Vector() {constant=<rgba 1 0 0 0>}))
  n64 = DotProduct(A=@n62.PS, B=Constant3Vector() {constant=<rgba 0 0 1 0>})
  n65 = LinearInterpolate(
   A=LinearInterpolate(A=Divide(A=@n59, B=@n60), B=Divide(A=@n61, B=@n60), Alpha=Step(X=@n63) {constY=0.5})
   B=Divide(A=ComponentMask(None=@n58) {g=true, r=true}, B=@n60)
   Alpha=Step(X=Abs(None=@n64)) {constY=0.5}
  )
  n66 = Floor(None=@n65)
  n67 = Abs(None=ComponentMask(None=MaterialFunctionCall(Seed=@n57, Vec2=@n66) {materialFunction=&MF_Hash}.Vec2) {r=true})
  n68 = Frac(None=@n65)
  n69 = LinearInterpolate(Alpha=Saturate(None=@n67)) {constA=0.25, constB=0.45}
  n70 = SphereMask(A=ComponentMask(None=@n68) {r=true}, B=Constant() {r=0.5}, Radius=@n69) {hardnessPercent=0}
  n71 = SphereMask(A=ComponentMask(None=@n68) {g=true}, B=Constant() {r=0.5}, Radius=@n69) {hardnessPercent=0}
  n72 = Multiply(A=Sign(None=@n70), B=Sign(None=@n71))
  n73 = LinearInterpolate(Alpha=Multiply(A=SmoothStep(Value=@n70) {constMax=0.5}, B=SmoothStep(Value=@n71) {constMax=0.5})) {constA=0.5}
  NamedRerouteDeclaration(None=Multiply(A=@n67, B=Multiply(A=@n72, B=@n73))) {name=CubeLayer01, nodeColor=<rgba 0 1 0.4141 1>}
  n74 = Multiply(A=@n31, B=@n37)
  NamedRerouteDeclaration(None=Power(Base=SmoothStep(Value=@n74) {constMax=0.5, constMin=0.1}) {constExponent=0.5}) {name=BlendAlpha01, nodeColor=<rgba 0 0.5547 1 1>}
 ]
}
